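# Sound cue for player foley when grabbing a heavy pickaxe handle, played at low volume in the melee weapon sound class.
# A Random node picks one of three equally weighted grab waves and feeds a Quality Level node with only its first slot connected.
# A Modulator at default settings follows, then a close-range pickaxe attenuation asset, and the melee weapon concurrency set governs simultaneous plays.
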SoundCue /Game/Sounds/Fort_Player/Hands/PickaxeGrab/Cues/PlayerFoley_PickaxeHandle_Heavy_Cue
root=Attenuation_0
nodes [
 {id=Attenuation_0, children [Modulator_0], properties {attenuationSettings=/Game/Sounds/Attenuation/Fort_Weapon_Pickaxe_Close_Att}}
 {id=Modulator_0, children [QualityLevel_0]}
 {id=QualityLevel_0, children [Random_0,null,null,null]}
 {id=Random_0, children [WavePlayer_9, WavePlayer_10, WavePlayer_11], properties {weights [1,1,1]}}
 {id=WavePlayer_10, properties {soundWaveAssetPtr=/Game/Sounds/Fort_Player/Hands/PickaxeGrab/Waves/Foley_Pickaxe_Heavy_Grab_02}}
 {id=WavePlayer_11, properties {soundWaveAssetPtr=/Game/Sounds/Fort_Player/Hands/PickaxeGrab/Waves/Foley_Pickaxe_Heavy_Grab_03}}
 {id=WavePlayer_9, properties {soundWaveAssetPtr=/Game/Sounds/Fort_Player/Hands/PickaxeGrab/Waves/Foley_Pickaxe_Heavy_Grab_01}}
]
properties {
 maxDistance=1536
 soundClassObject=&Fort_WeaponMelee_PSM
 volumeMultiplier=0.1
 concurrencySet [&Player_Melee_Weapon_Conc]
}
